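# Animation Blueprint for the RoosterRoast pickaxe mesh, adding secondary physics motion to its skeleton, including the eyes.
# The input pose runs through a RigidBody node simulated relative to the two-handed handle bone.
# Two chained Anim Dynamics nodes then spring the left and right eye bones relative to b_AR_Root, allowing limited linear travel.
# Each update, a nativized state function computes gravity override and rigid-body settings from a parameter asset and stores them in variables.
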AnimBlueprint /BRCosmetics/Weapons/FORT_Melee/Pickaxe_RoosterRoast/Meshes/RoosterRoast_Axe_AnimBP
target_skeleton=&/RoosterRoast_Axe_Skeleton
graphs [
 AnimGraph {nodes [
   AnimGraphNode_SubInput_0 = Input Pose()
   n1 = RigidBody!(ComponentPose=Local To Component(LocalPose=@AnimGraphNode_SubInput_0), Alpha, OverrideWorldGravity:Vector="0,0,-980", ExternalForce) -> (Pose) {
    bOverrideWorldGravity=true
    simulationSpace=BaseBoneSpace
    baseBoneRef {boneName=Melee_TwoHanded_Handle}
    overrideWorldGravity {z=-980}
    simSpaceSettings {
     maxAngularAcceleration=40
     maxAngularVelocity=8
     maxLinearAcceleration=1800
     maxLinearVelocity=700
     worldAlpha=1
     externalLinearDragV=<0.3 0.3 0.3>
     externalLinearVelocity=<0.3 0.3 0.3>
    }
   }
   n2 = Anim Dynamics - Bone: dyn_eye_00_left(ComponentPose=@n1, Alpha=1, ExternalForce="0,0,0") {
    angularDampingOverride=1
    angularSpringConstant=500
    bAngularSpring=true
    bLinearSpring=true
    bOverrideAngularDamping=true
    bOverrideLinearDamping=true
    bUseGravityOverride=true
    gravityScale=0.35
    linearDampingOverride=0.999
    linearSpringConstant=200
    simulationSpace=BoneRelative
    alphaBoolBlend {blendOption=Linear}
    boundBone {boneName=dyn_eye_00_left}
    componentAppliedLinearAccClamp=<3000 3000 3000>
    componentLinearAccScale=<0.3 0.1 0.1>
    componentLinearVelScale=<0.3 0.1 0.1>
    gravityOverride {z=-980}
    physicsBodyDefinitions [{
      boundBone {boneName=dyn_eye_00_left}
      boxExtents=<5 5 5>
      constraintSetup {angularLimitsMax=<0 0 0>, angularLimitsMin=<0 0 0>, angularTarget=<1 0 0>, linearAxesMax=<0 0.6 0.6>, linearAxesMin=<0 -0.7 -0.7>}
      localJointOffset=<0 0 0>
     }]
    relativeSpaceBone {boneName=b_AR_Root}
   }
   n3 = Anim Dynamics - Bone: dyn_eye_00_right(ComponentPose=@n2, Alpha=1, ExternalForce="0,0,0") {
    angularDampingOverride=1
    angularSpringConstant=500
    bAngularSpring=true
    bLinearSpring=true
    bOverrideAngularDamping=true
    bOverrideLinearDamping=true
    bUseGravityOverride=true
    gravityScale=0.35
    linearDampingOverride=0.999
    linearSpringConstant=200
    simulationSpace=BoneRelative
    alphaBoolBlend {blendOption=Linear}
    boundBone {boneName=dyn_eye_00_right}
    componentAppliedLinearAccClamp=<3000 3000 3000>
    componentLinearAccScale=<0.3 0.1 0.1>
    componentLinearVelScale=<0.3 0.1 0.1>
    gravityOverride {z=-980}
    physicsBodyDefinitions [{
      boundBone {boneName=dyn_eye_00_right}
      boxExtents=<5 5 5>
      constraintSetup {angularLimitsMax=<0 0 0>, angularLimitsMin=<0 0 0>, angularTarget=<1 0 0>, linearAxesMax=<0 0.4 0.5>, linearAxesMin=<0 -0.4 -0.5>}
      localJointOffset=<0 0 0>
     }]
    relativeSpaceBone {boneName=b_AR_Root}
   }
   Output Pose(Result=Component To Local(ComponentPose=@n3)) {layerGroup=None}
  ]}
 EventGraph {nodes [
   Event BlueprintUpdateAnimation()
   > n1 = "Gravity Override State Function (Parameter Asset) (Nativized)"(
    CharacterPartAnimInstance=Self-Reference()
    GameTimeInSeconds=GetGameTimeInSeconds()
    Parameters=&/M_MED_RoosterRoast_pickaxe_Parameters
   )
   > Set GravityOverride(GravityOverride=@n1)
   > Set RBAN Settings Out_0(RBAN Settings Out_0:Fort Rigid Body Anim Node Parameters Structure=@n1.OutRigidBodyAnimNodeParameters) -> (Output_Get:Fort Rigid Body Anim Node Parameters Structure)
   TryGetPawnOwner()
  ]}
]
variables [
 Gravity Override Params {type=GravityOverrideParamsStruct}
 GravityOverride {type=Vector}
 RBAN Settings Out_0 {type=FortRigidBodyAnimNodeParameters}
]
