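# Physics asset for the Rooster Roast two-handed pickaxe mesh, simulating a dangling coffee-cup part on the weapon.
# Capsule bodies cover the handle, a small cup base and two dynamic cup bones.
# The base is rotationally locked to the handle and anchors a chain of the two dynamic bones.
# Each link in that chain has its main swing limit on a different axis, with almost no twist allowance.
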PhysicsAsset /BRCosmetics/Weapons/FORT_Melee/Pickaxe_RoosterRoast/Meshes/RoosterRoast_Axe_Physics
bodies {
 Melee_TwoHanded_Handle [{
   shapeType=Capsule
   length=42.3054
   radius=21.3587
   shapeName=Melee_TwoHanded_Handle_capsule
   center=<-0.0008 5.3095 28.4189>
   rotation=<rot -17.8075 -90 0.0041>
  }]
 base_coffeecup_01_cr [{shapeType=Capsule, length=0.505, radius=0.505, shapeName=base_coffeecup_01_cr_capsule, rotation=<rot 0 -90 0>}]
 dyn_coffeecup_01_cr [{shapeType=Capsule, length=8.5, radius=4.6, shapeName=dyn_coffeecup_01_cr_capsule, center=<8 0 0>, rotation=<rot 90 0 0>}]
 dyn_coffeecup_02_cr [{shapeType=Capsule, length=8.5, radius=4.6, shapeName=dyn_coffeecup_01_cr_capsule, center=<8 0 0>, rotation=<rot 90 0 0>}]
}
constraints [
 {bone1Name=dyn_coffeecup_01_cr, bone2Name=base_coffeecup_01_cr, swing1LimitDegrees=0, swing2LimitDegrees=45, twistLimitDegrees=1}
 {
  bone1Name=base_coffeecup_01_cr
  bone2Name=Melee_TwoHanded_Handle
  swing1LimitDegrees=45
  swing1Motion=Locked
  swing2LimitDegrees=45
  swing2Motion=Locked
  twistLimitDegrees=45
  twistMotion=Locked
 }
 {bone1Name=dyn_coffeecup_02_cr, bone2Name=dyn_coffeecup_01_cr, swing1LimitDegrees=60, swing2LimitDegrees=0, twistLimitDegrees=0}
]
